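# Translucent, unlit, two-sided material for the storm player-damage visual, usable on mesh particles.
# A Fresnel term blends EdgeColor and InnerColor into an emissive adjustment function.
# Panning noise drives both a vertex-normal offset term and an opacity term.
# That opacity multiplies the noise by a V-coordinate mask fading both ends, a second Fresnel, depth fade and an Opacity parameter.
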Material /FNGameplayCues/VisualThreatManager/StormVisuals/Materials/M_StormPlayerDamage
graph {
 properties {
  bUsedWithMeshParticles=true
  blendMode=BLEND_Translucent
  pixelDepthOffsetMode=PDOM_Legacy
  shadingModel=MSM_Unlit
  twoSided=true
  usedShadingModels=Unlit
 }
 nodes [
  Add(
   A=Subtract(
    A=MaterialFunctionCall() {materialFunction=&LocalPosition}.Local Position
    B=WorldPosition().XYZ
   )
   B=MaterialFunctionCall() {materialFunction=&ComponentPivotLocation}.Result
  )
  n1 = TextureSample(UVs=Panner(Coordinate=TextureCoordinate() {uTiling=2, vTiling=1.2}) {speedX=0.05, speedY=0.33}) {texture=&T_NoisePack_01}
  Add(A=Multiply(A=ConstantBiasScale(None=@n1.R) {bias=-0.5, scale=15}, B=VertexNormalWS())) {constB=0}
  n2 = Constant() {r=2}
  MaterialFunctionCall(Emissive=LinearInterpolate(
    A=VectorParameter() {parameterName=EdgeColor, defaultValue=<rgba 0.6878 0.2361 1 0>}.RGB
    B=VectorParameter() {parameterName=InnerColor, defaultValue=<rgba 0.2443 0.2126 1 0>}.RGB
    Alpha=MaterialFunctionCall(Exponent=@n2) {materialFunction=&MF_Fresnel_FX}.Fresnel
   )) {materialFunction=&MF_FX_EmissiveAdjustments}
  n3 = ComponentMask(None=TextureCoordinate()) {g=true}
  n4 = Power(Base=Multiply(A=@n3, B=OneMinus(None=@n3))) {constExponent=3}
  n5 = Multiply(A=@n1.R, B=Saturate(None=Multiply(A=@n4) {constB=128}))
  n6 = MaterialFunctionCall(Exponent=Constant() {r=7}) {materialFunction=&MF_Fresnel_FX}
  n7 = Constant() {r=12}
  Multiply(
   A=Multiply(A=Multiply(A=@n5, B=@n6.Fresnel) {constA=1}, B=DepthFade(FadeDistance=@n7))
   B=ScalarParameter() {defaultValue=0.15, parameterName=Opacity}
  )
 ]
}
